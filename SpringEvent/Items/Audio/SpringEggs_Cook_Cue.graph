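# Sound cue for cooking a Spring Event egg item, reusing the FireJar Molotov cook sounds.
# A Random node picks between two wave players with equal weight.
# It plays in the gadget items sound class with gadget-place attenuation, a slightly lowered volume and a concurrency override that caps simultaneous instances at a low count.
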SoundCue /SpringEvent/Items/Audio/SpringEggs_Cook_Cue
root=Random_1
nodes [
 {id=Random_1, children [WavePlayer_0,WavePlayer_5], properties {weights [1,1]}}
 {id=WavePlayer_0, properties {soundWaveAssetPtr=/Game/Sounds/Fort_Weapons/Grenades/FireJar/Molotov_Cook_01}}
 {id=WavePlayer_5, properties {soundWaveAssetPtr=/Game/Sounds/Fort_Weapons/Grenades/FireJar/Molotov_Cook_02}}
]
properties {
 attenuationSettings=/Game/Sounds/Attenuation/Fort_GadgetPlace_Att
 bOverrideConcurrency=true
 maxDistance=2000
 soundClassObject=&Fort_Gadgets_Items
 volumeMultiplier=0.9
 concurrencyOverrides {maxCount=2}
}
