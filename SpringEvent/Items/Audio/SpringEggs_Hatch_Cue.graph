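# SoundCue for the hatching sound of SpringEvent egg items, classed under gadget and item audio.
# A single wave player for the SpringEggs_Hatch wave feeds a fixed half-second Delay node at the root, so playback starts slightly after the trigger.
# An egg-hatch gadget attenuation asset shapes spatial falloff, and a concurrency override caps simultaneous instances at two.
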SoundCue /SpringEvent/Items/Audio/SpringEggs_Hatch_Cue
root=Delay_0
nodes [
 {id=Delay_0, children [WavePlayer_1], properties {delayMax=0.5, delayMin=0.5}}
 {id=WavePlayer_1, properties {soundWaveAssetPtr=&/SpringEggs_Hatch}}
]
properties {
 attenuationSettings=/Game/Sounds/Attenuation/Fort_GadgetActivate_EggHatch_Att
 bOverrideConcurrency=true
 maxDistance=4000
 soundClassObject=&Fort_Gadgets_Items
 concurrencyOverrides {maxCount=2}
}
