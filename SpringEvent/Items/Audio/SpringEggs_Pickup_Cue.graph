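# Sound cue for the SpringEvent egg pickup, built by layering two sounds reused from other items.
# A mixer plays the Halloween GumDrop pickup immediately alongside a slightly delayed Low Gravity Stone jump sound.
# The delayed layer is pitched up by a fixed-value modulator and faded out by an enveloper volume curve.
# The cue uses the weapon pickup attenuation and the loot collect sound class.
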SoundCue /SpringEvent/Items/Audio/SpringEggs_Pickup_Cue
root=Mixer_0
nodes [
 {id=Delay_0, children [Enveloper_0], properties {delayMax=0.175, delayMin=0.175}}
 {
  id=Enveloper_0
  children [Modulator_0]
  properties {
   pitchCurve {editorCurveData {keys [{time=0}]}}
   volumeCurve {editorCurveData {keys [<at 0 1>,<at 0.2101 0.831>,<at 0.4233 0.2297>,<at 1.2187 -0.035>]}}
  }
 }
 {id=Mixer_0, children [WavePlayer_2,Delay_0], properties {inputVolume [1,1]}}
 {id=Modulator_0, children [WavePlayer_0], properties {pitchMax=1.3, pitchMin=1.3}}
 {id=WavePlayer_0, properties {soundWaveAssetPtr=/Game/Sounds/Fort_GamePlay_Sounds/Player/LowGravityStones/LowGravityStone_Jump_01}}
 {id=WavePlayer_2, properties {soundWaveAssetPtr=/Game/Athena/Sounds/Items/S14_Halloween/GumDrop_Halloween_Pickup_01}}
]
properties {
 attenuationSettings=&Fort_Weapon_Pickup_Att
 maxDistance=3500
 soundClassObject=&Fort_Loot_Collect
}
