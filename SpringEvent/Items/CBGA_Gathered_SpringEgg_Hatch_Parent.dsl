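; Parent Blueprint for gatherable SpringEvent eggs that can hatch into chickens unless gathered first.
; After a hotfixable minimum time, player proximity triggers a server chance roll, then the Living World Manager spawns the chicken early in stasis.
; Replicated hatching state drives shake, audio and client cosmetics while a server timer releases the chicken and expires the egg.
; Gathering mid-hatch destroys the pending chicken and suppresses hatch effects.
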(event EventBeginPlay
  ;[0 1 0.1174] Event BeginPlay
  ; Check for hatching feature, roll if this egg is elibible, and set a minimum time before hatching can occur
  (|CallParentFunction)
  (Networking|SwitchHasAuthority
    (:Authority
      (if (ScalableFloat|GetValueasBool (Variables|Hotfixes|GetSF_IsHatchEnabled))
        ;;[0.3353 0.6433 1] Have a minimum time after the egg is spawned before it can try to hatch
        (Utilities|Time|SetTimerbyEvent (EventDispatchers|CreateEvent :!event Custom|EnableEggToHatch) (ScalableFloat|GetValueatLevel (Variables|Hotfixes|GetSF_MinTimeBeforeHatching)))))
    (:Remote
      ;;[0.503 0 1] Play egg laying VFX/SFX - works because all eggs are spawned from chickens
      (GameplayCue|ExecuteGameplayCue%28Burst%29Local self (Variables|VFX|GetGCN%20Egg%20Laying) (Ability|GameplayCue|MakeGameplayCueParameters 0.0)))))

(event Custom|EnableEggToHatch
  ;[0 1 0.1174] EnableEggToHatch
  ; The egg has reached the minimum lifetime duration to hatch, enable it or hatch now
  (Variables|Gameplay|Hatching|SetbIsReadyToHatch true)
  ;;[0.3353 0.6433 1] Check if there is a current player overlap to hatch immediately
  (if (CallFunction|CheckforPlayerPawnsinHatchRadius)
    (if (CallFunction|RolltoBeginHatching)
      (CallFunction|SpawnChickenfromHatchedEgg) ; chicken will spawn in "Stasis" until it is activated
    )))

(event OnComponentBeginOverlap(PlayerHatchArea)
  (params (OverlappedComponent :type Primitive%20Component%20Object%20Reference) (OtherActor :type Actor%20Object%20Reference) (OtherComp :type Primitive%20Component%20Object%20Reference) (OtherBodyIndex :type Integer) (bFromSweep :type Boolean) (SweepResult :type Hit%20Result%20Structure%20%28by%20ref%29))
  ;[0.04191 0 1] OnComponentBeginOverlap
  ; Start hatchingwhen a player enters the radius of the egg and all hatching parameters are met (minimum time before hatching and not already hatching)
  ;;[0.3353 0.6433 1] Ready but not already hatching
  (if (and (Variables|Gameplay|Hatching|GetbIsReadyToHatch) (not (Variables|Gameplay|Hatching|GetbIsHatching)))
    (Utilities|Casting|CastToFortPlayerPawn OtherActor)
    (if (CallFunction|RolltoBeginHatching)
      (CallFunction|SpawnChickenfromHatchedEgg) ; chicken will spawn in "Stasis" until it is activated
    )))

(event Custom|SpawnChickenFromHatchedEgg
  (bind _returnvalue (Class|FortAthenaLivingWorldManager|GetLivingWorldManager))
  ;[0.04191 0 1] SpawnChickenFromHatchedEgg
  ; The chicken is spawned early and set into a "stasis" state where so the living world manager can allocate the resources and ensure it exists at the end of the hatching animation
  (Utilities|IsValid _returnvalue
    (:Is%20Valid
      ;;[0.04191 0 1] Try to spawn the spring chicken of the same type as the egg - set in the child class defaults
      (Class|FortAthenaLivingWorldManager|TrySpawnEvent _returnvalue (Utilities|Struct|MakeDataTableRowHandle "/Lager/Lager/EventTables/Lager_EventTable_NugS_Eggs.Lager_EventTable_NugS_Eggs" (Variables|Gameplay|Hatching|GetChickenToHatchRowName)) (Math|Transform|MakeTransform (Transformation|GetActorLocation self) (Math|Rotator|MakeRotator 0.0 0.0 (Math|Random|RandomFloatinRange 0.0 3.6e+02)) "1.000000,1.000000,1.000000") (EventDispatchers|CreateEvent :!event Custom|SpringNugSpawnedInStasis)))
    (:Is%20Not%20Valid
      ;;[1 0 0.9849] Invalid Living World Manager - unexpected error
      (Fortnite|FortShippingLog self "SpringEgg: Living World Manager is invalid, so Nug cannot be spawned" true))))

(event Custom|BeginHatchingEgg_Cosmetics
  ;[0.503 0 1] BeginHatchingEgg_Cosmetics
  ; Play a hatching animation to communicate that the egg will become a chicken if it isn't collected/consumed first
  (Networking|CanExecuteCosmeticEvents
    (:True
      (CallFunction|MaterialPlayEggAnim 0.25)
      (Utilities|FlowControl|Delay (ScalableFloat|GetValueatLevel (Variables|Hotfixes|GetSF_HatchingTellDuration)))
      (CallFunction|FinishHatchingEggCosmetics))))

(event Custom|SpringNugSpawnedInStasis
  (params (SpawnedActors :type Array%20of%20Actor%20Object%20References) (bSuccess :type Boolean))
  ;[0.3353 0.6433 1] SpringNugSpawnedInStasis
  ; The chicken was spawned in stasis and the hatching process should continue, otherwise it can try again later
  (if bSuccess
    ;;[0.3353 0.6433 1] Setting a reference of the spawned chicken and passing a reference to self to the it
    (bind _asnpc_pawn_irwin_prey_nug_s_base (Utilities|Casting|CastToNPC_Pawn_Irwin_Prey_NugS_Base (Utilities|Array|Get(acopy) SpawnedActors)))
    (Variables|Gameplay|Hatching|SetSpawnedChicken _asnpc_pawn_irwin_prey_nug_s_base)
    (Variables|/SpringEvent/NugS/Pawns/NPC_Pawn_Irwin_Prey_NugS_Base.NPC_Pawn_Irwin_Prey_NugS_Base_C|SetEggInstigator self _asnpc_pawn_irwin_prey_nug_s_base)
    (CallFunction|BeginHatchingEgg)))

(event Custom|BeginHatchingEgg
  ;[0 1 0.1174] BeginHatchingEgg
  ; Replicate the egg hatching to the client for cosmetics and set the hatching animation timer
  (Variables|Gameplay|Hatching|SetbIsHatching true)
  (Utilities|Time|SetTimerbyEvent (EventDispatchers|CreateEvent :!event Custom|FinishHatchingEgg) (ScalableFloat|GetValueatLevel (Variables|Hotfixes|GetSF_HatchingTellDuration))) ; start timer to know when the hatching animation is completed
)

(event Custom|FinishHatchingEgg
  ;[1 0.3396 0] FinishHatchingEgg
  ; Disable interaction, remove the chicken from stasis, and set life span to destroy the egg object
  (Variables|Gameplay|Hatching|SetbIsHatched true)
  (CallFunction|TryRemoveHatchedChickenfromStasis)
  (Actor|SetLifeSpan 0.5))

(event Custom|FinishHatchingEgg_Cosmetics
  ;[0.503 0 1] FinishHatchingEgg_Cosmetics
  ; Play hatching VFX and turn off the egg
  (Rendering|SetVisibility (Variables|Default|GetStaticMesh) false true)
  (if (not (Variables|Default|GetPreventHatchingCosmetics))
    ;;[0.5 0.5 0.5] TODO: Should be replaced with GCNBurst
    (Effects|Components|ParticleSystem|SpawnFXAtLocation :FXSystemAsset (Variables|VFX|GetVFX_Hatched) :Location (Transformation|GetActorLocation self) :Scale "0.200000,0.200000,0.200000" :bAutoDestroy true :PoolingMethod "None" :bAutoActivate true)))

(event Custom|TryRemoveHatchedChickenFromStasis
  (bind _spawnedchicken (Variables|Gameplay|Hatching|GetSpawnedChicken))
  ;[0.04191 0 1] TryRemoveHatchedChickenFromStasis
  ; Delegate call for the
  (if (Variables|Gameplay|Hatching|GetbIsHatched)
    (Default|CallEventChickenHatched)
    (Variables|/SpringEvent/NugS/Pawns/NPC_Pawn_Irwin_Prey_NugS_Base.NPC_Pawn_Irwin_Prey_NugS_Base_C|SetPhasedIn true _spawnedchicken)
    (Ability|GameplayEffect|AddLooseGameplayTags _spawnedchicken (GameplayTags|MakeGameplayTagContainerfromTag "(TagName=\"SpringChicken.JustHatched\")") true)
    (Development|PrintString "SpringEgg: Hatch completed, remove chicken from stasis" true true "(R=0.000000,G=0.162341,B=1.000000,A=1.000000)" 8.0)))

(event Interaction|EventOnGatherOrInteract
  (params (InteractingPawn :type Fort%20Pawn%20Object%20Reference) (InteractionBeingAttempted :type EInteractionBeingAttempted%20Enum :default "FirstInteraction"))
  (bind _spawnedchicken (Variables|Gameplay|Hatching|GetSpawnedChicken))
  ;[0.2969 0.1037 0] EventOnGatherOrInteract
  ; Destroy the chicken if it was in the process of hatching
  (Networking|SwitchHasAuthority
    (:Authority
      (if (not (Variables|Gameplay|Hatching|GetbIsHatched))
        (Utilities|IsValid _spawnedchicken
          (:Is%20Valid
            (Actor|DestroyActor _spawnedchicken)
            (Variables|Default|SetPreventHatchingCosmetics true)))))))

(event EventOnDeathPlayEffects
  (params (Damage :type Float%20%28single-precision%29) (DamageTags :type Gameplay%20Tag%20Container%20Structure%20%28by%20ref%29) (Momentum :type Vector :default "0, 0, 0") (HitInfo :type Hit%20Result%20Structure%20%28by%20ref%29) (InstigatedBy :type Fort%20Pawn%20Object%20Reference) (DamageCauser :type Actor%20Object%20Reference) (EffectContext :type Gameplay%20Effect%20Context%20Handle%20Structure))
  ;[0.6236 0 1] Event OnDeathPlayEffects (Client)
  ; Executes a GameplayAbility in addition to any parent effects. GCN should likely not be used with the default "Death Particles (Cascade)" - added in order to support Niagra VFX
  (|CallParentFunction Damage DamageTags Momentum HitInfo InstigatedBy DamageCauser EffectContext)
  (GameplayCue|ExecuteGameplayCue%28Burst%29Local self (Variables|Death|GetGCNBurst_OnDeath) (Ability|GameplayCue|MakeGameplayCueParameters :NormalizedMagnitude 0.0 :RawMagnitude 0.0 :EffectContext (Class|AthenaFunctionLibrary|MakeSimpleFortEffectContext InstigatedBy DamageCauser) :AggregatedTargetTags DamageTags :Location (Transformation|GetActorLocation self) :Normal "0.000000,0.000000,1.000000" :Instigator InstigatedBy :EffectCauser DamageCauser :SourceObject self :GameplayEffectLevel 1 :AbilityLevel 1 :bReplicateLocationWhenUsingMinimalRepProxy false)))

(fn BlueprintCanInteract
  (params (InteractingPawn :type Fort%20Pawn%20Object%20Reference) (InteractionBeingAttempted :type EInteractionBeingAttempted%20Enum :default "FirstInteraction") (InteractionType :type TInteractionType%20Enum :default "IT_NoInteraction"))
  (return (and (|CallParentFunction InteractingPawn InteractionBeingAttempted InteractionType) (not (Variables|Gameplay|Hatching|GetbIsHatched)))))

(fn CheckForPlayerPawnsInHatchRadius
  (for _array_element (Collision|GetOverlappingActors (Variables|Default|GetPlayerHatchArea) 0 :!class /Script/Engine.PrimitiveComponent)
    (Utilities|Casting|CastToFortPlayerPawn _array_element)
    (return true))
  (return false))

(fn MaterialPlayEggAnim
  (params (Ease%20Duration :type Float%20%28double-precision%29 :default 0.25) (Stop%20Shake :type Boolean))
  (Rendering|Material|SetCustomPrimitiveDataVector3 (Variables|Default|GetStaticMesh) 0 (Math|Vector|MakeVector (Utilities|Time|GetGameTimeinSeconds) Ease%20Duration (Math|Conversions|ToFloat(Boolean) (not Stop%20Shake)))))

(fn OnRep_bIsHatching
  (if (Variables|Gameplay|Hatching|GetbIsHatching)
    (Networking|CanExecuteCosmeticEvents
      (:True
        (CallFunction|BeginHatchingEggCosmetics) ; running as an event so it can use a Timeline animation
        (Audio|Components|Audio|Play (Variables|Default|GetEggHatching) 0.0)))))

(fn OnRep_EggOverflowDespawn
  (bind _bishatching (Variables|Gameplay|Hatching|GetbIsHatching))
  (Networking|SwitchHasAuthority
    (:Authority
      (if (not _bishatching)
        (Actor|SetLifeSpan 0.5)))
    (:Remote
      (if (not _bishatching)
        (GameplayCue|ExecuteGameplayCue%28Burst%29Local self (Variables|VFX|GetGCN%20Egg%20Despawning) (Ability|GameplayCue|MakeGameplayCueParameters 0.0))))))

(fn OnRep_PreventHatchingCosmetics
  (if (Variables|Gameplay|Hatching|GetbIsHatched)
    _)
  (Audio|Components|Audio|FadeOut (Variables|Default|GetEggHatching) 0.2 0.0 "Linear" :!class /Script/Engine.AudioComponent))

(fn RollToBeginHatching
  ;[0.3353 0.6433 1] RNG Roll to determine if this egg is able to hatch
  ;;[0.9743 0.4736 1] Return if roll % is met
  (return (<= (Math|Random|RandomFloat) (ScalableFloat|GetValueatLevel (Variables|Hotfixes|GetSF_HatchRollChancePct)))))

(fn ConstructionScript
  (|CallParentFunction))
