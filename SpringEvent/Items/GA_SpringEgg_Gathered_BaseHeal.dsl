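; Gameplay ability that heals a player who gathers a Spring event egg, spilling surplus healing into shield.
; On activation it compares missing health to a scalable base heal amount, capping health gain and converting the remainder to shield.
; Each portion of at least one point applies the configured effect to the owner via SetByCaller magnitudes.
; On ending, it removes the infinite effect that granted it.
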(event Ability|EventActivateAbility
  (|CallParentFunction)
  (bind _outfortplayerpawn (CallFunction|GetAvatarasValidFortPlayerPawn)
    (:ValidFortPlayerPawn
      (Variables|Default|SetPlayerPawn _outfortplayerpawn)
      (CallFunction|GrantEffectiveHealing)
      (Ability|EndAbility))))

(event Ability|EventFailedToActivatePassiveAbility
  (params (ActorInfo :type Gameplay%20Ability%20Actor%20Info%20Structure))
  (|CallParentFunction ActorInfo))

(event Ability|EventOnEndAbility
  (params (bWasCancelled :type Boolean))
  ;[1 0.2576 0] OnEndAbility
  ; cleanup
  (|CallParentFunction bWasCancelled)
  (Ability|RemoveGrantedbyEffect) ; remove infinite GE that granted this ability
)

(fn GrantEffectiveHealing
  (bind _healthdifference (Variables|Default|GetHealthDifference))
  (bind _returnvalue (- (ScalableFloat|GetValueatLevel (Variables|Default|GetSF_BaseEggHealingAmount)) _healthdifference))
  (bind _returnvalue_2 (>= _returnvalue 0.0))
  (bind _sf_baseegghealingamount (Variables|Default|GetSF_BaseEggHealingAmount))
  (bind _returnvalue_3 (ScalableFloat|GetValueatLevel _sf_baseegghealingamount))
  (bind _returnvalue_4 (select _returnvalue_2 _returnvalue 0.0))
  (bind _ge_effecttoapply (Variables|Default|GetGE_EffectToApply))
  (bind _playerpawn (Variables|Default|GetPlayerPawn)
    (:then
      (Variables|Default|SetHealthDifference (- (FortPawn|GetMaxHealth _playerpawn) (FortPawn|GetHealth _playerpawn)))
      (Variables|Default|SetHealthHealAmount (Math|Float|Min(Float) _healthdifference _returnvalue_3))
      (Variables|Default|SetShieldHealAmount _returnvalue_4)
      (Utilities|FlowControl|Sequence
        (:then_0
          (if (>= (- (FortPawn|GetMaxHealth _playerpawn) (FortPawn|GetHealth _playerpawn)) 1.0)
            (bind _returnvalue_5 (Ability|GameplayEffect|AssignTagSetbyCallerMagnitude (Ability|MakeOutgoingGameplayEffectSpec _ge_effecttoapply) "(TagName=\"SetByCaller.Healing\")" (Variables|Default|GetHealthHealAmount)))
            (Ability|ApplyGameplayEffectSpecToOwner _returnvalue_5)))
        (:then_1
          (if (>= _returnvalue_4 1.0)
            (bind _returnvalue_6 (Ability|GameplayEffect|AssignTagSetbyCallerMagnitude (Ability|MakeOutgoingGameplayEffectSpec _ge_effecttoapply) "(TagName=\"SetByCaller.Shield\")" (Variables|Default|GetShieldHealAmount)))
            (Ability|ApplyGameplayEffectSpecToOwner _returnvalue_6)))))
    (:else)))
